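; Deimos rift spawner Blueprint: a damageable building actor that spawns AI and takes a damage effect per spawn.
; On begin play the server applies damage reduction and can destroy overlapping vehicles and EQS-teleport players from its no-build zone.
; Health thresholds set a replicated DamageState whose OnRep drives dynamic materials, particles and audio.
; Cosmetic state events and gameplay cues handle spawn, ramp-up and dying effects.
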(event EventBeginPlay
  (bind _sm_cuberunes (Variables|Default|GetSM_CubeRunes))
  ;[0.002934 1 0.03942] EVENT -> BeginPlay
  (CallFunction|EnableSpawning (ScalableFloat|GetValueasBool (Variables|Default|GetEnableDeimosSpawning)))
  (Utilities|FlowControl|Sequence
    (:then_0
      (EventDispatchers|BindEventtoOnDamaged (EventDispatchers|CreateEvent :!event Custom|OnDamaged_Bind))
      (EventDispatchers|BindEventtoOnBuildingRiftSpawnedAI (EventDispatchers|CreateEvent :!event Custom|OnBuildingRiftSpawnedAI_Bind))
      (Audio|Components|Audio|SetSound (Variables|Default|GetDeimos_Spawner_Loop_Sound) (Variables|Default|GetSound_DeimosSpawnerLoop))
      (Audio|Components|Audio|FadeIn (Variables|Default|GetDeimos_Spawner_Loop_Sound) 3.0 1.0 0.0 "Linear" :!class /Script/Engine.AudioComponent))
    (:then_1
      (Networking|SwitchHasAuthority
        (:Authority
          (GameplayEffects|ApplyGameplayEffectToSelf :self (Ability|GetAbilitySystemComponent self) :GameplayEffectClass (Variables|Default|GetGE_DamageReduction) :Level 0.0)
          (if (Variables|Default|GetDoSpawnEffects)
            (Networking|FlushNetDormancy)
            (CallFunction|SpawnEffects)))))
    (:then_2
      (Networking|SwitchHasAuthority
        (:Authority
          ;;[1 1 1] Destroy unneeded components on server
          (Components|DestroyComponent (Variables|Default|GetPointLight))
          (if (Utilities|IsValid _sm_cuberunes :!class /Script/Engine.KismetSystemLibrary)
            (Components|DestroyComponent _sm_cuberunes)
            ; If we spawn mid-match, make sure we adjust any player position within our bounds
            (if (ScalableFloat|GetValueasBool (Variables|Default|GetEnablePlayerTeleportOnSpawn))
              (bind _outactors (Collision|BoxOverlapActors (Transformation|GetWorldLocation (Variables|Default|GetNoBuildZone)) (Components|Box|GetScaledBoxExtent (Variables|Default|GetNoBuildZone)) (Utilities|Array|MakeArray "ObjectTypeQuery3") "/Script/FortniteGame.FortPlayerPawnAthena" :!class /Script/Engine.KismetSystemLibrary))
              (for _array_element _outactors
                (bind _asfort_player_pawn_athena (Utilities|Casting|CastToFortPlayerPawnAthena _array_element))
                (if (Vehicle|IsinVehicle _asfort_player_pawn_athena)
                  (Vehicle|ForceDestroyVehicle (Vehicle|GetVehicle _asfort_player_pawn_athena) true)
                  (else
                    _))
                (bind _returnvalue (AI|EQS|RunEQSQuery "/Figment_Deimos/Spawners/EQS_DeimosSpawner_PlayerAdjustment.EQS_DeimosSpawner_PlayerAdjustment" self "RandomBest25Pct"))
                (AI|EQS|AssignOnQueryFinished _returnvalue (delegate Custom|OnQueryFinishedEvent_Event))))))))))

(event Custom|OnQueryFinishedEvent_Event
  (params (QueryInstance :type EQS%20Query%20Instance%20Object%20Reference) (QueryStatus :type EEnvQueryStatus%20Enum :default "Processing"))
  ; If we spawn mid-match, make sure we adjust any player position within our bounds
  (if (Utilities|Enum|Equal(Enum) QueryStatus "Success")
    (bind _resultlocations (AI|EQS|GetQueryResultsasLocations QueryInstance))
    (Class|FortPawn|TeleportTo (stale EventBeginPlay _asfort_player_pawn_athena) (Utilities|Array|Get(acopy) _resultlocations) (Transformation|GetActorRotation (stale EventBeginPlay _asfort_player_pawn_athena)) false false)))

(event EventBP_Cosmetic_Intro
  ;[0.5937 1 0.5123] Replicated Cosmetic Events
  ; Calling SetCosmeticState replicates to these.
  (Development|PrintString "BP_DeimosRift: Cosmetic Intro" false false))

(event EventBP_Cosmetic_Idle
  ;[0.5937 1 0.5123] Replicated Cosmetic Events
  ; Calling SetCosmeticState replicates to these.
  (Development|PrintString "BP_DeimosRift: Cosmetic Idle" false false))

(event EventBP_Cosmetic_RampUp
  ;[0.5937 1 0.5123] Replicated Cosmetic Events
  ; Calling SetCosmeticState replicates to these.
  (Development|PrintString "BP_DeimosRift: Cosmetic Ramp Up" false false)
  (Rendering|Material|SetScalarParameterValue (Variables|Default|GetMaskedMID) "Spawn" (Utilities|Time|GetGameTimeinSeconds) :!class /Script/Engine.MaterialInstanceDynamic)
  (GameplayCue|ExecuteGameplayCue%28Burst%29Local self (Variables|Default|GetChargeUpFXGC) (Variables|Default|GetCueParamsChargeUp)))

(event EventBP_Cosmetic_ShouldDie
  (bind _returnvalue (Transformation|GetActorLocation))
  ;[0.5937 1 0.5123] Replicated Cosmetic Events
  ; Calling SetCosmeticState replicates to these.
  (Development|PrintString "BP_DeimosRift: Cosmetic Should Die" false false)
  (Variables|Default|SetDamageState 4)
  (Camera|PlayWorldCameraShake "/Game/Athena/Prototype/Blueprints/Cube/WarmupEventCameraShake.WarmupEventCameraShake_C" _returnvalue 5e+02 2.5e+03)
  (Audio|PlaySoundatLocation (Variables|Default|GetSound_StageThreeBuildup) _returnvalue)
  (Components|Activation|SetActive (Variables|Default|GetP_Deimos_Rift_Dying) true false)
  (Rendering|Material|SetScalarParameterValue (Variables|Default|GetCubeMatLOD1) "Dying" 1.0 :!class /Script/Engine.MaterialInstanceDynamic)
  (Rendering|Material|SetScalarParameterValue (Variables|Default|GetCubeMat) "Damage" 1.0 :!class /Script/Engine.MaterialInstanceDynamic))

(event BuildingRift|EventOnStartedEncounterSpawn
  (params (SpawnedAI :type Fort%20AIPawn%20Object%20Reference))
  ;[0.5937 1 0.5123] Replicated Cosmetic Events
  ; Calling SetCosmeticState replicates to these.
  (Development|PrintString "BP_DeimosRift: Spawn" false false)
  (GameplayCue|ExecuteGameplayCue%28Burst%29Local self (Variables|Default|GetFiendSpawnFXGC) (Ability|GameplayCue|MakeGameplayCueParameters :NormalizedMagnitude 0.0 :RawMagnitude 0.0 :EffectContext "()" :AggregatedSourceTags "(GameplayTags=)" :AggregatedTargetTags "(GameplayTags=)" :Location (Math|Vector|MakeVectorNetQuantize10 (.x (Transformation|GetActorLocation SpawnedAI)) (.y (Transformation|GetActorLocation SpawnedAI)) (.z (Transformation|GetActorLocation SpawnedAI))) :Normal "0,0,0" :GameplayEffectLevel 1 :AbilityLevel 1 :bReplicateLocationWhenUsingMinimalRepProxy false)))

(event EventOnDeathPlayEffects
  (params (Damage :type Float%20%28single-precision%29) (DamageTags :type Gameplay%20Tag%20Container%20Structure%20%28by%20ref%29) (Momentum :type Vector :default "0, 0, 0") (HitInfo :type Hit%20Result%20Structure%20%28by%20ref%29) (InstigatedBy :type Fort%20Pawn%20Object%20Reference) (DamageCauser :type Actor%20Object%20Reference) (EffectContext :type Gameplay%20Effect%20Context%20Handle%20Structure))
  ;[1 1 1] EVENT -> OnDeathPlayEffects
  (Audio|Components|Audio|FadeOut (Variables|Default|GetDeimos_Spawner_Loop_Sound) 1.0 0.0 "Linear" :!class /Script/Engine.AudioComponent)
  (Utilities|FlowControl|Sequence
    (:then_0
      (Components|Activation|SetActive (Variables|Default|GetP_Deimos_Rift_Dying) false false))
    (:then_1
      (Development|SetHiddeninGame (Variables|Transformation|GetRootComponent) true true))))

(event EventOnDeathServer
  (params (Damage :type Float%20%28single-precision%29) (DamageTags :type Gameplay%20Tag%20Container%20Structure%20%28by%20ref%29) (Momentum :type Vector :default "0, 0, 0") (HitInfo :type Hit%20Result%20Structure%20%28by%20ref%29) (InstigatedBy :type Controller%20Object%20Reference) (DamageCauser :type Actor%20Object%20Reference) (EffectContext :type Gameplay%20Effect%20Context%20Handle%20Structure))
  ;[1 1 1] EVENT -> OnDeathServer
  (GameplayEffects|ApplyGameplayEffectToSelf :self (Ability|GetAbilitySystemComponent self) :GameplayEffectClass (Variables|GameplayEffects|GetGE_Destroyed) :Level 0.0)
  (Development|PrintString "Death" false false))

(event Custom|OnDamaged_Bind
  (params (DamagedActor :type Actor%20Object%20Reference) (Damage :type Float%20%28single-precision%29) (InstigatedBy :type Controller%20Object%20Reference) (DamageCauser :type Actor%20Object%20Reference) (HitLocation :type Vector) (FHitComponent :type Primitive%20Component%20Object%20Reference) (BoneName :type Name) (Momentum :type Vector))
  ;[1 1 1] CUSTOM EVENT -> OnDamaged_Bind
  (Development|PrintString (Utilities|GetDisplayName DamageCauser))
  (CallFunction|UpdateDamageState (Health|GetHealthPercent)))

(event Custom|OnBuildingRiftSpawnedAI_Bind
  ;[1 1 1] CUSTOM EVENT -> OnBuildingRiftSpawnedAI_Bind
  (GameplayEffects|ApplyGameplayEffectToSelf :self (Ability|GetAbilitySystemComponent self) :GameplayEffectClass (Variables|Default|GetGE_DamagePerSpawn) :Level 0.0))

(event Custom|SpawnEffects
  ;[0.2572 1 0.819] CUSTOM EVENT -> SpawnEffects (made this a multicast event so it'll happen and if you miss it then it won't try to replay when the client comes into relevancy)
  ;;[0.8422 0.03439 1] This hides the spawn anim
  (bind _meshes (CallFunction|GetSpawnMeshes))
  (for _array_element _meshes
    (Rendering|Material|SetScalarParameterValueonMaterials _array_element "Spawn_CurrentTime" 2.7e+03))
  (GameplayCue|ExecuteGameplayCue%28Burst%29Local self (Variables|Default|GetSpawnFXGC) (Ability|GameplayCue|MakeGameplayCueParameters :NormalizedMagnitude 0.0 :RawMagnitude 0.0 :EffectContext "()" :AggregatedSourceTags "(GameplayTags=)" :AggregatedTargetTags "(GameplayTags=)" :Location "0,0,0" :Normal "0,0,0" :GameplayEffectLevel 1 :AbilityLevel 1 :bReplicateLocationWhenUsingMinimalRepProxy false))
  (Utilities|FlowControl|Delay 0.2)
  (CallFunction|StartDynamicRiftSpawnCosmetics))

(event Custom|StartDynamicRiftSpawnCosmetics
  ;[0.9316 0 1] StartDynamicRiftSpawnCosmetics
  (Networking|CanExecuteCosmeticEvents
    (:True
      (bind _meshes (CallFunction|GetSpawnMeshes))
      (for _array_element _meshes
        (Rendering|Material|SetScalarParameterValueonMaterials _array_element "Spawn_CurrentTime" (Utilities|Time|GetGameTimeinSeconds))
        (Rendering|Material|SetScalarParameterValueonMaterials _array_element "Spawn_AnimLength" 0.0)))))

(event Custom|UpdateDamageState
  (params (HealthPercent :type Float%20%28double-precision%29))
  ;[0.9726 0.3664 1] Update the damage state of this rift based on the health percentage (normalized 0 - 1). Executed on the server and DamageState replicates to the client. DamageState is used to update cosmetics.
  (if (<= HealthPercent 0.3)
    (Variables|Default|SetDamageState 3)
    (elif (<= HealthPercent 0.6)
      (Variables|Default|SetDamageState 2)
      (elif (<= HealthPercent 0.9)
        (Variables|Default|SetDamageState 1)
        (else
          (Variables|Default|SetDamageState 0))))))

(fn FlashCubeMaterial
  (params (Length :type Float%20%28double-precision%29))
  (bind _cubemat (Variables|Default|GetCubeMat))
  (Rendering|Material|SetScalarParameterValue _cubemat "Hit_CurrentTime" (Utilities|Time|GetGameTimeinSeconds) :!class /Script/Engine.MaterialInstanceDynamic)
  (Rendering|Material|SetScalarParameterValue _cubemat "Hit_AnimLength" Length :!class /Script/Engine.MaterialInstanceDynamic))

(fn GetSpawnMeshes
  (return (Utilities|Array|MakeArray (Variables|Default|GetCubeMesh) (Variables|Default|GetCosmeticBaseMesh))))

(fn OnRep_DamageState
  (bind _rootcomponent (Variables|Transformation|GetRootComponent))
  (bind _returnvalue (Transformation|GetWorldLocation _rootcomponent))
  (bind _damagestate (Variables|Default|GetDamageState))
  (bind _cubemat (Variables|Default|GetCubeMat))
  (bind _p_deimos_riftspawn_damagestate (Variables|Default|GetP_Deimos_RiftSpawn_DamageState))
  (Components|Activation|SetActive _p_deimos_riftspawn_damagestate (> _damagestate 0) true)
  (Effects|Components|ParticleSystem|SetFloatParameter _p_deimos_riftspawn_damagestate "Damage2" (Math|Conversions|ToFloat(Boolean) (>= _damagestate 2)))
  (Effects|Components|ParticleSystem|SetFloatParameter _p_deimos_riftspawn_damagestate "Damage3" (Math|Conversions|ToFloat(Boolean) (>= _damagestate 3)))
  ;[1 1 1] Flash the mesh during transition
  (if (> (Variables|Default|GetDamageState) 0)
    (CallFunction|FlashCubeMaterial (Variables|Default|GetHitAnimLength)))
  (Utilities|FlowControl|Sequence
    (:then_0
      (switch int (Variables|Default|GetDamageState)
        (:0
          (Rendering|Material|SetScalarParameterValue _cubemat "Damage" 0.0 :!class /Script/Engine.MaterialInstanceDynamic))
        (:1
          (Rendering|Material|SetScalarParameterValue _cubemat "Damage" 0.2 :!class /Script/Engine.MaterialInstanceDynamic)
          (Audio|PlaySoundatLocation (Variables|Default|GetSound_StageOneExplode) _returnvalue))
        (:2
          (Rendering|Material|SetScalarParameterValue _cubemat "Damage" 0.5 :!class /Script/Engine.MaterialInstanceDynamic)
          (Audio|PlaySoundatLocation (Variables|Default|GetSound_StageTwoExplode) _returnvalue))
        (:3
          (Rendering|Material|SetScalarParameterValue _cubemat "Damage" 0.8 :!class /Script/Engine.MaterialInstanceDynamic)
          (Audio|PlaySoundatLocation (Variables|Default|GetSound_StageTwoExplode) _returnvalue))))
    (:then_1
      (Audio|Parameter|SetIntegerParameter (Variables|Default|GetDeimos_Spawner_Loop_Sound) "Damage" (Variables|Default|GetDamageState) :!class /Script/AudioExtensions.AudioParameterControllerInterface))))

(fn ConstructionScript
  (bind _returnvalue (Rendering|Material|CreateDynamicMaterialInstance (Variables|Default|GetBaseMaterial) "None" "None" :!class /Script/Engine.KismetMaterialLibrary))
  (Variables|Default|SetCubeMat _returnvalue)
  (bind _returnvalue_2 (Rendering|Material|CreateDynamicMaterialInstance (Variables|Default|GetBaseLODMaterial) "None" "None" :!class /Script/Engine.KismetMaterialLibrary))
  (Variables|Default|SetCubeMatLOD1 _returnvalue_2)
  (Rendering|Material|SetMaterial (Variables|Default|GetCubeMesh) 0 (Variables|Default|GetCubeMat))
  (Rendering|Material|SetMaterial (Variables|Default|GetCubeMesh) 1 (Variables|Default|GetCubeMatLOD1))
  (bind _returnvalue_3 (Rendering|Material|CreateDynamicMaterialInstance (Rendering|Material|GetMaterial (Variables|Default|GetCosmeticBaseMesh) 0) "None" "None" :!class /Script/Engine.KismetMaterialLibrary))
  (bind _output_get (Variables|Default|SetBaseMeshMat _returnvalue_3))
  (Rendering|Material|SetMaterial (Variables|Default|GetCosmeticBaseMesh) 0 _output_get))
