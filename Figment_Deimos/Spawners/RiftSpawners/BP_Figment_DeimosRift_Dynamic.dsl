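; Dynamic variant of the Figment Deimos rift spawner that expires after a set lifetime unless it keeps taking damage.
; The authority starts a lifetime timer on BeginPlay, and server-side damage restarts it, throttled by a minimum interval.
; On expiry it applies a lifetime-lapsed gameplay effect to itself and destroys the actor.
; Construction adjusts the damage-state particle parameters and removes the cube-runes mesh, while cosmetic state events only add debug prints to parent behavior.
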(event EventBeginPlay
  ;[1 1 1] EVENT -> BeginPlay
  (|CallParentFunction)
  (Networking|SwitchHasAuthority
    (:Authority
      (CallFunction|SetLifeTimeTimer))))

(event Custom|SetLifeTimeTimer
  ;[1 1 1] CUSTOM EVENT -> SetLifeTimeTimer
  (Utilities|Time|ClearandInvalidateTimerbyHandle (Variables|Default|GetLifeTimeTimerHandle))
  (bind _returnvalue (Utilities|Time|SetTimerbyEvent (EventDispatchers|CreateEvent :!event Custom|LifeTimeExpired) (Variables|Default|GetLifeTime)))
  (Variables|Default|SetLifeTimeTimerHandle _returnvalue))

(event EventOnDamageServer
  (params (Damage :type Float%20%28single-precision%29) (DamageTags :type Gameplay%20Tag%20Container%20Structure%20%28by%20ref%29) (Momentum :type Vector :default "0, 0, 0") (HitInfo :type Hit%20Result%20Structure%20%28by%20ref%29) (InstigatedBy :type Controller%20Object%20Reference) (DamageCauser :type Actor%20Object%20Reference) (EffectContext :type Gameplay%20Effect%20Context%20Handle%20Structure))
  ;[1 1 1] EVENT -> OnDamageServer
  (if (> (Utilities|TimeSince (Variables|Default|GetTimeLastSetLifeTimeTimer)) (Variables|Default|GetThrottleTime))
    (Variables|Default|SetTimeLastSetLifeTimeTimer (Utilities|Time|GetTimeSeconds))
    (CallFunction|SetLifeTimeTimer)))

(event Custom|LifeTimeExpired
  ;[1 1 1] CUSTOM EVENT -> LifeTimeExpired
  (GameplayEffects|ApplyGameplayEffectToSelf :self (Ability|GetAbilitySystemComponent self) :GameplayEffectClass (Variables|GameplayEffects|GetGE_LifeTimeLapsed) :Level 0.0)
  (Actor|DestroyActor))

(event EventBP_Cosmetic_Intro
  ;[0.5937 1 0.5123] Replicated Cosmetic Events
  ; Calling SetCosmeticState replicates to these.
  (|CallParentFunction)
  (Development|PrintString "BP_DeimosRift_Dynamic: Cosmetic Intro" false false))

(event EventBP_Cosmetic_Idle
  ;[0.5937 1 0.5123] Replicated Cosmetic Events
  ; Calling SetCosmeticState replicates to these.
  (|CallParentFunction)
  (Development|PrintString "BP_DeimosRift_Dynamic: Cosmetic Idle" false false))

(event EventBP_Cosmetic_RampUp
  ;[0.5937 1 0.5123] Replicated Cosmetic Events
  ; Calling SetCosmeticState replicates to these.
  (|CallParentFunction)
  (Development|PrintString "BP_DeimosRift_Dynamic: Cosmetic Ramp Up" false false))

(event EventBP_Cosmetic_ShouldDie
  ;[0.5937 1 0.5123] Replicated Cosmetic Events
  ; Calling SetCosmeticState replicates to these.
  (|CallParentFunction)
  (Development|PrintString "BP_DeimosRift_Dynamic: Cosmetic Should Die" false false))

(fn OnRep_ServerSpawnTime)

(fn ConstructionScript
  (bind _p_deimos_riftspawn_damagestate (Variables|Default|GetP_Deimos_RiftSpawn_DamageState))
  (|CallParentFunction)
  (Effects|Components|ParticleSystem|SetVectorParameter _p_deimos_riftspawn_damagestate "SizeScale" "0.700000,0.700000,0.700000")
  (Effects|Components|ParticleSystem|SetFloatParameter _p_deimos_riftspawn_damagestate "Radius" 0.0)
  (Components|DestroyComponent (Variables|Default|GetSM_CubeRunes)))
